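# Unlit masked two-sided material for Niagara mesh particles in the StylizedStrike drift trail, reusing that boost's cone texture.
# Static and panned samples blend, with red splitting bright areas from colored ones and alpha times particle alpha forming opacity.
# A panning cosine gradient from purple to blue supplies color, replaced by a flat value when painted, then tinted by particle color.
# The graph holds two near-identical chains; one feeds MF_FX_EmissiveAdjustments.
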Material /VehicleCosmetics/DriftTrail/DriftTrail_StylizedStrike/Material/M_DriftTrail_StylizedStrike_BaseColor
graph {
 properties {
  bUsedWithNiagaraMeshParticles=true
  blendMode=BLEND_Masked
  pixelDepthOffsetMode=PDOM_Legacy
  shadingModel=MSM_Unlit
  twoSided=true
  usedShadingModels=Unlit
 }
 nodes [
  DynamicParameter()
  n1 = ParticleColor()
  n2 = TextureSample() {texture=/VehicleCosmetics/Boosts/Boost_StylizedStrike/Texture/T_Boost_StylizedStrike_Cone_Diffuse}
  n3 = LinearInterpolate(A=@n2.R, B=@n2.R) {constAlpha=1}
  n4 = TextureSample(UVs=Panner(Coordinate=TextureCoordinate(), Time=Constant() {r=1})) {texture=&T_Boost_StylizedStrike_Cone_Diffuse}
  n5 = LinearInterpolate(A=@n4.R, B=@n4.R) {constAlpha=1}
  n6 = Saturate(None=Multiply(A=Constant() {r=10}, B=LinearInterpolate(A=@n3, B=@n5) {constAlpha=1}))
  n7 = Saturate(None=OneMinus(None=@n6))
  n8 = DynamicParameter() {parameterIndex=1}
  n9 = Constant3Vector() {constant=<rgba 0.2 0 0.8 0>}
  n10 = Multiply(A=Constant3Vector() {constant=<rgba 0 0.3 0.8 0>}, B=Constant() {r=5})
  n11 = Panner(Coordinate=TextureCoordinate() {uTiling=0.3, vTiling=0.1}) {speedX=0.8}
  n12 = Power(Base=Cosine(None=ComponentMask(None=@n11) {r=true}), Exp=Constant() {r=3.5})
  n13 = Multiply(
   A=If(A=@n8.Param1, B=Constant() {r=1}, "A > B"=@n8.Param1, "A == B"=@n8.Param1, "A < B"=LinearInterpolate(A=@n9, B=@n10, Alpha=@n12)) {constB=0.99}
   B=@n6
  ) {constA=0.1}
  MaterialFunctionCall(Emissive=Multiply(A=@n1.RGB, B=Add(A=@n7, B=@n13))) {materialFunction=&MF_FX_EmissiveAdjustments}
  n14 = LinearInterpolate(A=@n2.A, B=@n4.A) {constAlpha=1}
  n15 = Constant() {r=5}
  Multiply(A=@n1.A, B=Saturate(None=Multiply(A=@n14, B=@n15)))
  MaterialExpressionMeshEmitterVertexColor_3 = ParticleColor()
  n16 = TextureSample() {texture=&T_Boost_StylizedStrike_Cone_Diffuse}
  n17 = TextureSample(UVs=Panner(Coordinate=TextureCoordinate(), Time=Constant() {r=1})) {texture=&T_Boost_StylizedStrike_Cone_Diffuse}
  MaterialExpressionMeshEmitterDynamicParameter_1 = DynamicParameter() {defaultValue=<rgba 1 0 1 1>, paramNames [Param1,isPainted,BlackCore,Param4]}
  MaterialExpressionConstantClamp_15 = Saturate(None=Multiply(A=LinearInterpolate(A=@n16.A, B=@n17.A, Alpha=@MaterialExpressionMeshEmitterDynamicParameter_1.Param1), B=Constant() {r=5}))
  Multiply(A=@MaterialExpressionMeshEmitterVertexColor_3.A, B=@MaterialExpressionConstantClamp_15)
  n18 = LinearInterpolate(A=@n16.R, B=@n16.R, Alpha=@MaterialExpressionMeshEmitterDynamicParameter_1.BlackCore)
  n19 = LinearInterpolate(A=@n17.R, B=@n17.R, Alpha=@MaterialExpressionMeshEmitterDynamicParameter_1.BlackCore)
  MaterialExpressionConstantClamp_14 = Saturate(None=Multiply(A=Constant() {r=10}, B=LinearInterpolate(A=@n18, B=@n19, Alpha=@MaterialExpressionMeshEmitterDynamicParameter_1.Param1)))
  MaterialExpressionConstantClamp_13 = Saturate(None=OneMinus(None=@MaterialExpressionConstantClamp_14))
  n20 = Constant() {r=0.4}
  n21 = Constant3Vector() {constant=<rgba 0 0.3 0.8 0>}
  n22 = Constant() {r=5}
  n23 = TextureCoordinate() {uTiling=0.3, vTiling=0.1}
  n24 = Cosine(None=ComponentMask(None=Panner(Coordinate=@n23) {speedX=0.8}) {r=true})
  n25 = Constant() {r=3.5}
  n26 = If(
   A=Floor(None=Saturate(None=@MaterialExpressionMeshEmitterDynamicParameter_1.isPainted))
   "A > B"=@n20
   "A == B"=@n20
   "A < B"=LinearInterpolate(A=Constant3Vector() {constant=<rgba 0.2 0 0.8 0>}, B=Multiply(A=@n21, B=@n22), Alpha=Power(Base=@n24, Exp=@n25))
  ) {constB=1}
  Multiply(
   A=@MaterialExpressionMeshEmitterVertexColor_3.RGB
   B=Add(A=@MaterialExpressionConstantClamp_13, B=Multiply(A=@n26, B=@MaterialExpressionConstantClamp_14) {constA=0.45})
  )
 ]
}
